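# Control Rig for the M_MED_James face accessory that rotates the beak's base bone as the head bends relative to the upper spine.
# A sequence first runs a spherical pose reader on the head, parented to spine_04, and stores its falloff output in neck_bend_alpha.
# That alpha then weights a Modify Transforms rotation on base_beak, which carries the dyn_beak child.
# The reader's debug drawing is enabled, and Neck_bend and Targeting_alpha are unused.
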ControlRigBlueprint /BRCosmetics/Characters/Player/Male/Medium/Bodies/M_MED_James/Meshes/Parts/M_MED_James_FaceAcc_CtrlRig
graphs [
 RigVMFunctionLibrary {links [], nodes []}
 RigVMModel {nodes [
   RigUnit_BeginExecution = &RigUnit_BeginExecution()
   RigVMFunction_Sequence = &RigVMFunction_Sequence(ExecuteContext=@RigUnit_BeginExecution.ExecutePin)
   n1 = &RigUnit_SphericalPoseReader(
    ActiveRegionScaleFactors=(PositiveWidth=3,NegativeWidth=0.1,PositiveHeight=6,NegativeHeight=6)
    ActiveRegionSize=0.1
    Debug=(bDrawDebug=True,DebugScale=25,DebugSegments=20,DebugThickness=0.25)
    DriverAxis=(X=1,Y=0,Z=0)
    DriverItem=(Type=Bone,Name="head")
    FalloffRegionScaleFactors=(PositiveWidth=-3,NegativeWidth=1,PositiveHeight=0,NegativeHeight=0)
    FalloffSize=0.25
    OptionalParentItem=(Type=Bone,Name="spine_04")
    RotationOffset=(X=90,Y=0,Z=-180)
    ExecutePin=@RigVMFunction_Sequence.A
   )
   > RigVMVariableNode(Value=@n1.OutputParam) {variable=neck_bend_alpha}
   &RigUnit_ModifyTransforms(
    ItemToModify=((Item=(Type=Bone,Name="base_beak"),Transform=(Rotation=(X=0,Y=0,Z=0.5736,W=0.8192),Translation=(X=0,Y=0,Z=0),Scale3D=(X=1,Y=1,Z=1))))
    ExecutePin=@RigVMFunction_Sequence.B
    Weight=RigVMVariableNode() {variable=neck_bend_alpha}.Value
   )
  ]}
]
hierarchy [
 root {loc=<0 0 0>, rot=<0 0 0 1>}
 pelvis {parent=root, loc=<0 -0.7071 93.1191>, rot=<0.0127 0.707 -0.0127 -0.707>}
 spine_01 {parent=pelvis, loc=<6.7494 -0.6388 0>, rot=<0 0 0.034 -0.9994>}
 spine_02 {parent=spine_01, loc=<7.2518 -0.035 0>, rot=<0 0 0.0238 -0.9997>}
 spine_03 {parent=spine_02, loc=<6.0104 0 0>, rot=<0 0 0.0149 -0.9999>}
 spine_04 {parent=spine_03, loc=<8.1024 0 0>, rot=<0 0 -0.0132 -0.9999>}
 spine_05 {parent=spine_04, loc=<10.0146 0 0>, rot=<0 0 -0.0015 -1>}
 neck_01 {parent=spine_05, loc=<11.3138 0.1529 0>, rot=<0 0 -0.2003 -0.9797>}
 neck_02 {parent=neck_01, loc=<4.5627 0 0>, rot=<0 0 0.0504 -0.9987>}
 head {parent=neck_02, loc=<5.3914 0 0>, rot=<0 0 0.115 -0.9934>}
 base_hair {parent=head, loc=<10.8234 8.4479 0>, rot=<0 0 -0.3007 -0.9537>}
 dyn_hair {parent=base_hair, loc=<0 0 0>, rot=<0 0 0 1>}
 beak {parent=head, loc=<4.3694 -6.2492 0>, rot=<0 0 0.4617 -0.887>}
 hair {parent=head, loc=<10.8234 8.4479 0>, rot=<0 0 -0.3007 -0.9537>}
 base_beak {parent=head, loc=<4.3694 -6.2492 0>, rot=<0 0 0.4617 -0.887>}
 dyn_beak {parent=base_beak, loc=<0 0 0>, rot=<0 0 0 1>}
]
variables [
 Neck_bend {type=double, default=0}
 Targeting_alpha {type=double, default=0}
 neck_bend_alpha {type=double, default=0}
]
